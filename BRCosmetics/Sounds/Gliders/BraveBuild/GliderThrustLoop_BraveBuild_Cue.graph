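# Looping thrust sound cue for the BraveBuild glider, with glider-specific attenuation and sound class.
# A mixer blends five looping layers (low, wind, main, turbine and turbine wind) at individual levels, with one input left empty.
# Each layer has a continuous modulator driven by the AnyDirection parameter, which raises pitch slightly as it increases.
# The same parameter fades the low layer out while the other layers get louder.
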SoundCue /BRCosmetics/Sounds/Gliders/BraveBuild/GliderThrustLoop_BraveBuild_Cue
root=Mixer_0
nodes [
 {
  id=Mixer_0
  children [ModulatorContinuous_0, null, ModulatorContinuous_2, ModulatorContinuous_3, ModulatorContinuous_4, ModulatorContinuous_5]
  properties {inputVolume [0.9,0.6,0.5,0.9,1,0.85]}
 }
 {
  id=ModulatorContinuous_0
  children [WavePlayer_0]
  properties {
   pitchModulationParams {default=0, maxOutput=1.15, minOutput=1, parameterName=AnyDirection}
   volumeModulationParams {default=0, maxOutput=0, minOutput=0.5, parameterName=AnyDirection}
  }
 }
 {
  id=ModulatorContinuous_2
  children [WavePlayer_2]
  properties {
   pitchModulationParams {default=0, maxOutput=1.1, minOutput=1, parameterName=AnyDirection}
   volumeModulationParams {default=0, maxOutput=0.5, minOutput=0.3, parameterName=AnyDirection}
  }
 }
 {
  id=ModulatorContinuous_3
  children [WavePlayer_3]
  properties {
   pitchModulationParams {default=0, maxOutput=1.15, minOutput=1, parameterName=AnyDirection}
   volumeModulationParams {default=0, maxOutput=0.65, parameterName=AnyDirection}
  }
 }
 {
  id=ModulatorContinuous_4
  children [WavePlayer_4]
  properties {
   pitchModulationParams {default=0, maxOutput=1.1, minOutput=1, parameterName=AnyDirection}
   volumeModulationParams {default=0, minOutput=0.8, parameterName=AnyDirection}
  }
 }
 {
  id=ModulatorContinuous_5
  children [WavePlayer_5]
  properties {
   pitchModulationParams {default=0, maxOutput=1.1, minOutput=1, parameterName=AnyDirection}
   volumeModulationParams {default=0, maxOutput=0.8, minOutput=0.4, parameterName=AnyDirection}
  }
 }
 {id=WavePlayer_0, properties {bLooping=true, soundWaveAssetPtr=&/IP_Glider_BraveBuild_Thrust_LP_Low}}
 {id=WavePlayer_2, properties {bLooping=true, soundWaveAssetPtr=&/IP_Glider_BraveBuild_Thrust_LP_Wind}}
 {id=WavePlayer_3, properties {bLooping=true, soundWaveAssetPtr=&/IP_Glider_BraveBuild_Thrust_LP_Main}}
 {id=WavePlayer_4, properties {bLooping=true, soundWaveAssetPtr=&/IP_Glider_BraveBuild_Thrust_LP_Turbine}}
 {
  id=WavePlayer_5
  properties {bLooping=true, soundWaveAssetPtr=&/IP_Glider_BraveBuild_Thrust_LP_Turbine_Wind}
 }
]
properties {
 attenuationSettings=&Fort_Player_Glider_Thrust_Athena_Att
 audioPropertiesSheet=&APS_Cosmetic_Glider_Loop
 maxDistance=12500
 soundClassObject=&Fort_Gliders_Land
 volumeMultiplier=0.6
 concurrencySet [null]
}
